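# Physics asset for the Plum Retro backpack's dynamic cape, defining its bodies and joints for simulation.
# Left, middle and right capsule chains hang from the clavicle cape bones and backpack root, with tight twist and swing limits widening down the cape.
# Some side segments link to the middle chain through angularly free constraints.
# Torso, arm and leg capsules, a crouch sphere and a flat root box provide collision shapes.
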PhysicsAsset /BRCosmetics/Accessories/FORT_Backpacks/Backpack_M_MED_Plum_Retro/Meshes/M_MED_Plum_Retro_DynCape
bodies {
 backpack_root [{shapeType=Capsule, length=0.505, radius=0.505, rotation=<rot 0 -90 0>}]
 calf_l [{shapeType=Capsule, length=31.8125, radius=12.5363, center=<-31.4163 -2.6359 -0.0003>, rotation=<rot 90 0 -1.5716>}]
 calf_r [{shapeType=Capsule, length=31.8125, radius=12.5363, center=<31.4163 2.6359 0.0003>, rotation=<rot 90 0 -1.5716>}]
 cape_00_l [{shapeType=Capsule, length=1, radius=1, rotation=<rot 0 -90 0>}]
 cape_00_r [{shapeType=Capsule, length=1, radius=1, rotation=<rot 0 -90 0>}]
 cape_01_l [{shapeType=Capsule, length=8.25, radius=7.5, shapeName=cape_01_l_capsule, center=<5.5 -7 -3>, rotation=<rot 0 0 -10>}]
 cape_01_r [{shapeType=Capsule, length=8, radius=7.75, shapeName=cape_01_l_capsule, center=<5.5 -7 3>, rotation=<rot 0 0 15.0006>}]
 cape_02_l [{shapeType=Capsule, length=6.5, radius=10.75, shapeName=cape_02_r_capsule, center=<10 -10 0>, rotation=<rot 0 0 -0.0007>}]
 cape_02_r [{shapeType=Capsule, length=6.5, radius=10, shapeName=cape_02_r_capsule, center=<10 -10 0>, rotation=<rot 0 0 -10.0006>}]
 cape_03_l [{shapeType=Capsule, length=8.75, radius=10.5, shapeName=cape_03_l_capsule, center=<10 -6 0>, rotation=<rot 0 0 5.001>}]
 cape_03_r [{shapeType=Capsule, length=8.75, radius=10.5, shapeName=cape_03_r_capsule, center=<10 -6 2>}]
 cape_04_l [{shapeType=Capsule, length=8.75, radius=10.5, shapeName=cape_04_l_capsule, center=<10 -6 0>, rotation=<rot 0 0 5.0005>}]
 cape_04_r [{shapeType=Capsule, length=8.75, radius=10.5, shapeName=cape_04_r_capsule, center=<10 -6 2>}]
 cape_05_l [{shapeType=Capsule, length=8.75, radius=10.5, shapeName=cape_05_l_capsule, center=<10 -6 0>, rotation=<rot 0 0 5.001>}]
 cape_05_r [{shapeType=Capsule, length=8.75, radius=10.5, shapeName=cape_05_r_capsule, center=<10 -6 2>}]
 cape_06_l [{shapeType=Capsule, length=8.75, radius=10.5, shapeName=cape_06_l_capsule, center=<10 -6 0>, rotation=<rot 0 0 5.001>}]
 cape_06_r [{shapeType=Capsule, length=8.75, radius=10.5, shapeName=cape_06_r_capsule, center=<10 -6 2>}]
 cape_07_l [{shapeType=Capsule, length=8.75, radius=10.5, shapeName=cape_07_l_capsule, center=<10 -6 0>, rotation=<rot 0 0 5.0013>}]
 cape_07_r [{shapeType=Capsule, length=8.75, radius=10.5, shapeName=cape_07_r_capsule, center=<10 -6 2>}]
 cape_08_l [{shapeType=Capsule, length=8.75, radius=10.5, shapeName=cape_08_l_capsule, center=<10 -6 0>, rotation=<rot 0 0 5.0005>}]
 cape_08_r [{shapeType=Capsule, length=8.75, radius=10.5, shapeName=cape_08_r_capsule, center=<10 -6 2>}]
 cape_mid_01 [{shapeType=Capsule, length=26.25, radius=10, shapeName=cape_mid_01_l_capsule, center=<10 -8 0>}]
 cape_mid_02 [{shapeType=Capsule, length=28.75, radius=11.75, shapeName=cape_mid_02_l_capsule, center=<10 -8 0>}]
 cape_mid_03 [{shapeType=Capsule, length=32.75, radius=11.25, shapeName=cape_mid_03_l_capsule, center=<10 -8 0>}]
 cape_mid_04 [{shapeType=Capsule, length=32.75, radius=11.25, shapeName=cape_mid_04_l_capsule, center=<10 -8 0>}]
 cape_mid_05 [{shapeType=Capsule, length=32.75, radius=11.25, shapeName=cape_mid_05_l_capsule, center=<10 -8 0>}]
 cape_mid_06 [{shapeType=Capsule, length=32.75, radius=11.25, shapeName=cape_mid_06_l_capsule, center=<10 -8 0>}]
 cape_mid_07 [{shapeType=Capsule, length=32.75, radius=11.25, shapeName=cape_mid_07_l_capsule, center=<10 -8 0>}]
 clavicle_cape_l [{shapeType=Capsule, length=0.505, radius=0.505, rotation=<rot 0 -90 0>}]
 clavicle_cape_r [{shapeType=Capsule, length=0.505, radius=0.505, rotation=<rot 0 -90 0>}]
 clavicle_l [{shapeType=Capsule, length=8.1779, radius=9.5408, center=<11.8178 -0.4325 -1.0336>, rotation=<rot 82.9488 10.7032 176.8089>}]
 clavicle_r [{shapeType=Capsule, length=8.4279, radius=9.2908, center=<-11.5831 0.5626 2.3092>, rotation=<rot 87.2048 8.3673 -3.8592>}]
 dyn_crouch_collision [{shapeType=Sphere, radius=45.375, center=<-0.0006 6.727 6.9543>}]
 lowerarm_l [{shapeType=Capsule, length=34.1377, radius=8.13, center=<18.227 0.9178 -2.1609>, rotation=<rot 82.7121 -0.0002 3.2014>}]
 lowerarm_r [
  {shapeType=Capsule, length=28.5439, radius=7.63, center=<-14.5234 -1.1325 0.8591>, rotation=<rot 82.7121 -0.0002 3.2014>}
  {shapeType=Capsule, length=23.5938, radius=6.4062, center=<-31.9441 -1.2158 5.2412>, rotation=<rot 0 0 89.9994>}
 ]
 pelvis [
  {shapeType=Capsule, length=9.8125, radius=18.0363, center=<-5.4177 0 -0.0004>}
  {shapeType=Capsule, length=17.6562, radius=16.7238, center=<-21.9244 0 0.0008>}
 ]
 root [{shapeType=Box, extentX=231.0625, extentY=231.0625, extentZ=12.375, center=<0 0 -4.1865>}]
 spine_02 [{shapeType=Capsule, length=13.875, radius=13.9738, center=<0 1.2686 0.0001>}]
 spine_04 [{shapeType=Capsule, length=11.875, radius=13.25, center=<0.8861 0.9074 -0.0001>}]
 spine_05 [{shapeType=Capsule, length=14.875, radius=12.75, center=<-1.6849 -0.1197 0.0001>, rotation=<rot 0 -0.0001 0>}]
 thigh_l [{shapeType=Capsule, length=16.125, radius=18.5988, center=<-20.2154 -0.6265 3.1209>, rotation=<rot 90 0 0>}]
 thigh_r [{shapeType=Capsule, length=16.125, radius=18.5988, center=<20.2154 0.6265 -2.7018>, rotation=<rot 90 0 0>}]
 upperarm_l [{shapeType=Capsule, length=21.1779, radius=8.0408, center=<13.2584 -0.8854 -1.0226>, rotation=<rot 87.3404 12.8837 -172.3975>}]
 upperarm_r [{shapeType=Capsule, length=20.9279, radius=8.2908, center=<-13.3097 0.8088 -0.2017>, rotation=<rot 87.3409 12.8742 9.1674>}]
}
constraints [
 {bone1Name=cape_02_l, bone2Name=cape_01_l, swing1LimitDegrees=25, swing2LimitDegrees=20, twistLimitDegrees=2}
 {bone1Name=cape_03_l, bone2Name=cape_02_l, swing1LimitDegrees=25, swing2LimitDegrees=15, twistLimitDegrees=2}
 {bone1Name=cape_04_l, bone2Name=cape_03_l, swing1LimitDegrees=25, swing2LimitDegrees=30, twistLimitDegrees=2}
 {bone1Name=cape_05_l, bone2Name=cape_04_l, swing1LimitDegrees=45, swing2LimitDegrees=45, twistLimitDegrees=2}
 {bone1Name=cape_06_l, bone2Name=cape_05_l, swing1LimitDegrees=45, swing2LimitDegrees=45, twistLimitDegrees=2}
 {bone1Name=cape_07_l, bone2Name=cape_06_l, swing1LimitDegrees=45, swing2LimitDegrees=45, twistLimitDegrees=2}
 {bone1Name=cape_08_l, bone2Name=cape_07_l, swing1LimitDegrees=45, swing2LimitDegrees=45, twistLimitDegrees=2}
 {bone1Name=cape_01_l, bone2Name=cape_00_l, swing1LimitDegrees=30, swing2LimitDegrees=15, twistLimitDegrees=5}
 {bone1Name=cape_01_r, bone2Name=cape_00_r, swing1LimitDegrees=30, swing2LimitDegrees=10, twistLimitDegrees=5}
 {bone1Name=cape_02_r, bone2Name=cape_01_r, swing1LimitDegrees=25, swing2LimitDegrees=20, twistLimitDegrees=2}
 {bone1Name=cape_03_r, bone2Name=cape_02_r, swing1LimitDegrees=25, swing2LimitDegrees=15, twistLimitDegrees=2}
 {bone1Name=cape_04_r, bone2Name=cape_03_r, swing1LimitDegrees=25, swing2LimitDegrees=30, twistLimitDegrees=2}
 {bone1Name=cape_05_r, bone2Name=cape_04_r, swing1LimitDegrees=45, swing2LimitDegrees=45, twistLimitDegrees=2}
 {bone1Name=cape_06_r, bone2Name=cape_05_r, swing1LimitDegrees=45, swing2LimitDegrees=45, twistLimitDegrees=2}
 {bone1Name=cape_07_r, bone2Name=cape_06_r, swing1LimitDegrees=45, swing2LimitDegrees=45, twistLimitDegrees=2}
 {bone1Name=cape_08_r, bone2Name=cape_07_r, swing1LimitDegrees=45, swing2LimitDegrees=45, twistLimitDegrees=2}
 {
  bone1Name=clavicle_cape_l
  bone2Name=clavicle_l
  swing1LimitDegrees=45
  swing1Motion=Free
  swing2LimitDegrees=45
  swing2Motion=Free
  twistLimitDegrees=45
  twistMotion=Free
 }
 {
  bone1Name=clavicle_cape_r
  bone2Name=clavicle_r
  swing1LimitDegrees=45
  swing1Motion=Free
  swing2LimitDegrees=45
  swing2Motion=Free
  twistLimitDegrees=45
  twistMotion=Free
 }
 {
  bone1Name=backpack_root
  bone2Name=spine_05
  swing1LimitDegrees=45
  swing1Motion=Free
  swing2LimitDegrees=45
  swing2Motion=Free
  twistLimitDegrees=45
  twistMotion=Free
 }
 {
  bone1Name=spine_05
  bone2Name=spine_04
  swing1LimitDegrees=45
  swing1Motion=Free
  swing2LimitDegrees=45
  swing2Motion=Free
  twistLimitDegrees=45
  twistMotion=Free
 }
 {
  bone1Name=spine_04
  bone2Name=spine_02
  swing1LimitDegrees=45
  swing1Motion=Free
  swing2LimitDegrees=45
  swing2Motion=Free
  twistLimitDegrees=45
  twistMotion=Free
 }
 {
  bone1Name=spine_02
  bone2Name=pelvis
  swing1LimitDegrees=45
  swing1Motion=Free
  swing2LimitDegrees=45
  swing2Motion=Free
  twistLimitDegrees=45
  twistMotion=Free
 }
 {
  bone1Name=lowerarm_l
  bone2Name=upperarm_l
  swing1LimitDegrees=45
  swing1Motion=Free
  swing2LimitDegrees=45
  swing2Motion=Free
  twistLimitDegrees=45
  twistMotion=Free
 }
 {
  bone1Name=lowerarm_r
  bone2Name=upperarm_r
  swing1LimitDegrees=45
  swing1Motion=Free
  swing2LimitDegrees=45
  swing2Motion=Free
  twistLimitDegrees=45
  twistMotion=Free
 }
 {
  bone1Name=thigh_l
  bone2Name=pelvis
  swing1LimitDegrees=45
  swing1Motion=Free
  swing2LimitDegrees=45
  swing2Motion=Free
  twistLimitDegrees=45
  twistMotion=Free
 }
 {
  bone1Name=calf_l
  bone2Name=thigh_l
  swing1LimitDegrees=45
  swing1Motion=Free
  swing2LimitDegrees=45
  swing2Motion=Free
  twistLimitDegrees=45
  twistMotion=Free
 }
 {
  bone1Name=thigh_r
  bone2Name=pelvis
  swing1LimitDegrees=45
  swing1Motion=Free
  swing2LimitDegrees=45
  swing2Motion=Free
  twistLimitDegrees=45
  twistMotion=Free
 }
 {
  bone1Name=calf_r
  bone2Name=thigh_r
  swing1LimitDegrees=45
  swing1Motion=Free
  swing2LimitDegrees=45
  swing2Motion=Free
  twistLimitDegrees=45
  twistMotion=Free
 }
 {
  bone1Name=pelvis
  bone2Name=root
  swing1LimitDegrees=45
  swing1Motion=Free
  swing2LimitDegrees=45
  swing2Motion=Free
  twistLimitDegrees=45
  twistMotion=Free
 }
 {
  bone1Name=dyn_crouch_collision
  bone2Name=root
  swing1LimitDegrees=45
  swing1Motion=Free
  swing2LimitDegrees=45
  swing2Motion=Free
  twistLimitDegrees=45
  twistMotion=Free
 }
 {bone1Name=cape_00_r, bone2Name=clavicle_cape_r, swing1LimitDegrees=45, swing2LimitDegrees=30, twistLimitDegrees=15}
 {
  bone1Name=cape_00_l
  bone2Name=clavicle_cape_l
  swing1LimitDegrees=45
  swing1Motion=Free
  swing2LimitDegrees=45
  swing2Motion=Free
  twistLimitDegrees=45
  twistMotion=Free
 }
 {
  bone1Name=upperarm_l
  bone2Name=clavicle_l
  swing1LimitDegrees=45
  swing1Motion=Free
  swing2LimitDegrees=45
  swing2Motion=Free
  twistLimitDegrees=45
  twistMotion=Free
 }
 {bone1Name=clavicle_r, bone2Name=spine_05, swing1LimitDegrees=45, swing2LimitDegrees=30, twistLimitDegrees=15}
 {
  bone1Name=clavicle_l
  bone2Name=spine_05
  swing1LimitDegrees=45
  swing1Motion=Free
  swing2LimitDegrees=45
  swing2Motion=Free
  twistLimitDegrees=45
  twistMotion=Free
 }
 {bone1Name=upperarm_r, bone2Name=clavicle_r, swing1LimitDegrees=45, swing2LimitDegrees=30, twistLimitDegrees=15}
 {bone1Name=cape_mid_01, bone2Name=backpack_root, swing1LimitDegrees=25, swing2LimitDegrees=15, twistLimitDegrees=2}
 {bone1Name=cape_mid_02, bone2Name=cape_mid_01, swing1LimitDegrees=25, swing2LimitDegrees=15, twistLimitDegrees=2}
 {bone1Name=cape_mid_03, bone2Name=cape_mid_02, swing1LimitDegrees=25, swing2LimitDegrees=30, twistLimitDegrees=2}
 {bone1Name=cape_mid_04, bone2Name=cape_mid_03, swing1LimitDegrees=45, swing2LimitDegrees=45, twistLimitDegrees=2}
 {bone1Name=cape_mid_05, bone2Name=cape_mid_04, swing1LimitDegrees=45, swing2LimitDegrees=45, twistLimitDegrees=2}
 {bone1Name=cape_mid_06, bone2Name=cape_mid_05, swing1LimitDegrees=45, swing2LimitDegrees=45, twistLimitDegrees=2}
 {bone1Name=cape_mid_07, bone2Name=cape_mid_06, swing1LimitDegrees=45, swing2LimitDegrees=45, twistLimitDegrees=2}
 {
  bone1Name=cape_06_r
  bone2Name=cape_mid_05
  swing1LimitDegrees=45
  swing1Motion=Free
  swing2LimitDegrees=45
  swing2Motion=Free
  twistLimitDegrees=45
  twistMotion=Free
 }
 {
  bone1Name=cape_07_r
  bone2Name=cape_mid_06
  swing1LimitDegrees=45
  swing1Motion=Free
  swing2LimitDegrees=45
  swing2Motion=Free
  twistLimitDegrees=45
  twistMotion=Free
 }
 {
  bone1Name=cape_05_l
  bone2Name=cape_mid_04
  swing1LimitDegrees=45
  swing1Motion=Free
  swing2LimitDegrees=45
  swing2Motion=Free
  twistLimitDegrees=45
  twistMotion=Free
 }
 {
  bone1Name=cape_06_l
  bone2Name=cape_mid_05
  swing1LimitDegrees=45
  swing1Motion=Free
  swing2LimitDegrees=45
  swing2Motion=Free
  twistLimitDegrees=45
  twistMotion=Free
 }
 {
  bone1Name=cape_07_l
  bone2Name=cape_mid_06
  swing1LimitDegrees=45
  swing1Motion=Free
  swing2LimitDegrees=45
  swing2Motion=Free
  twistLimitDegrees=45
  twistMotion=Free
 }
 {
  bone1Name=cape_03_r
  bone2Name=cape_mid_02
  swing1LimitDegrees=45
  swing1Motion=Free
  swing2LimitDegrees=45
  swing2Motion=Free
  twistLimitDegrees=45
  twistMotion=Free
 }
]
