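# Physics asset for the dynamic collar on the Plum Retro backpack cosmetic, covering five collar bones: one back and two per side.
# Each collar bone has a capsule body constrained to a tiny capsule on its matching MSTR bone, with moderate swing and tight twist limits.
# Clavicle and double head capsules add body shapes around the collar, with no constraints linking them to it.
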PhysicsAsset /BRCosmetics/Accessories/FORT_Backpacks/Backpack_M_MED_Plum_Retro/Meshes/M_MED_Plum_Retro_DynCollar
bodies {
 clavicle_l [{shapeType=Capsule, length=5.9132, radius=7.1427, center=<22.1957 -0.5417 -2.1721>, rotation=<rot 0 -8.6222 89.9999>}]
 clavicle_r [{shapeType=Capsule, length=8.6809, radius=4.5275, center=<-8.8172 2.0695 -7.7643>, rotation=<rot 0 10.3502 -84.9746>}]
 dyn_collar_01_l [{shapeType=Capsule, length=5, radius=5, shapeName=dyn_collar_01_l_capsule, center=<6 0 3.5>, rotation=<rot 0 0 79.9991>}]
 dyn_collar_01_l_MSTR [{shapeType=Capsule, length=0.505, radius=0.505, rotation=<rot 0 -90 0>}]
 dyn_collar_01_r [{shapeType=Capsule, length=5, radius=5, shapeName=dyn_collar_01_r_capsule, center=<6 0 3.5>, rotation=<rot 0 0 95.0003>}]
 dyn_collar_01_r_MSTR [{shapeType=Capsule, length=0.505, radius=0.505, rotation=<rot 0 -90 0>}]
 dyn_collar_02_l [{shapeType=Capsule, length=10.5, radius=5, shapeName=dyn_collar_02_l_capsule, center=<10 2 3>, rotation=<rot 0.0001 29.999 100.0001>}]
 dyn_collar_02_l_MSTR [{shapeType=Capsule, length=0.505, radius=0.505, rotation=<rot 0 -90 0>}]
 dyn_collar_02_r [{shapeType=Capsule, length=10.5, radius=5, shapeName=dyn_collar_02_l_capsule, center=<10 -2 3>, rotation=<rot 0 -30 84.9998>}]
 dyn_collar_02_r_MSTR [{shapeType=Capsule, length=0.505, radius=0.505, rotation=<rot 0 -90 0>}]
 dyn_collar_bk [{shapeType=Capsule, length=5, radius=5, shapeName=dyn_collar_bk_capsule, center=<6 0 3.5>, rotation=<rot 0 0 90>}]
 dyn_collar_bk_MSTR [{shapeType=Capsule, length=0.505, radius=0.505, rotation=<rot 0 -90 0>}]
 head [
  {shapeType=Capsule, length=3.75, radius=7.5, center=<9.0157 0.5984 -0.0002>, rotation=<rot 90 0 0>}
  {shapeType=Capsule, length=4.75, radius=5.25, center=<4.7418 6.152 -0.0002>, rotation=<rot 90 0 0>}
 ]
}
constraints [
 {bone1Name=dyn_collar_bk, bone2Name=dyn_collar_bk_MSTR, swing1LimitDegrees=30, swing2LimitDegrees=30, twistLimitDegrees=5}
 {bone1Name=dyn_collar_01_r, bone2Name=dyn_collar_01_r_MSTR, swing1LimitDegrees=30, swing2LimitDegrees=30, twistLimitDegrees=5}
 {bone1Name=dyn_collar_02_r, bone2Name=dyn_collar_02_r_MSTR, swing1LimitDegrees=30, swing2LimitDegrees=30, twistLimitDegrees=5}
 {bone1Name=dyn_collar_01_l, bone2Name=dyn_collar_01_l_MSTR, swing1LimitDegrees=30, swing2LimitDegrees=30, twistLimitDegrees=5}
 {bone1Name=dyn_collar_02_l, bone2Name=dyn_collar_02_l_MSTR, swing1LimitDegrees=30, swing2LimitDegrees=30, twistLimitDegrees=5}
]
