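# Animation Blueprint for the M_MED_Plum_Retro backpack, adding secondary motion to its cape and collar over the pose copied from the attached parent mesh.
# With rigid body enabled, RigidBody nodes simulate cape and collar, after bone offsets driven by crouch and arm pose from the event graph.
# Otherwise, Look At and Trail controller chains take over, with a separate chain selected by the DBNO state.
# The Control Rig node is not connected to the output.
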AnimBlueprint /BRCosmetics/Accessories/FORT_Backpacks/Backpack_M_MED_Plum_Retro/Meshes/M_MED_Plum_Retro_Pack_AnimBP
target_skeleton=&/M_MED_Plum_Retro_Pack_Skeleton
graphs [
 AnimGraph {nodes [
   Control Rig!(Source, Alpha, isSkydiving:double=Get SkydiveFloat() -> (SkydiveFloat:float)) -> (Pose) {
    alphaBoolBlend {blendOption=Linear}
    controlRigAssetReference {blueprintRigClass=/BRCosmetics/Accessories/FORT_Capes/Cape_M_MED_HighTower_Date/Meshes/M_MED_HighTower_Date_Cape_CtrlRig.M_MED_HighTower_Date_Cape_CtrlRig_C}
   }
   n1 = Get CapeGravityOverride() -> (CapeGravityOverride:Vector)
   RigidBody!(ComponentPose, Alpha, OverrideWorldGravity:Vector=@n1, ExternalForce) -> (Pose) {
    bClampLinearTranslationLimitToRefPose=true
    bOverrideWorldGravity=true
    overridePhysicsAsset=&/M_MED_Plum_Retro_DynCape
    simulationSpace=BaseBoneSpace
    alphaBoolBlend {blendOption=Linear}
    baseBoneRef {boneName=root}
    componentAppliedLinearAccClamp=<3000 3000 3000>
    componentLinearAccScale=<0.7 0.5 0.5>
    componentLinearVelScale=<0.7 0.5 0.5>
    overrideWorldGravity {z=-980}
   }
   n2 = Look At - Bone: cape_mid_02(ComponentPose=Local To Component(LocalPose=Use cached pose 'initialPoseCache'()), Alpha=1) {
    alphaBoolBlend {blendOption=Linear}
    boneToModify {boneName=cape_mid_02}
    lookAtLocation=<0 -25>
    lookAtTarget {boneReference {boneName=pelvis}}
    lookAt_Axis {axis=<1 0>}
   }
   n3 = Look At - Bone: cape_02_l(ComponentPose=@n2, Alpha=1) {
    alphaBoolBlend {blendOption=Linear}
    boneToModify {boneName=cape_02_l}
    lookAtLocation=<0 -25 -21>
    lookAtTarget {boneReference {boneName=pelvis}}
    lookAt_Axis {axis=<1 0>}
   }
   n4 = Look At - Bone: cape_02_r(ComponentPose=@n3, Alpha=1) {
    alphaBoolBlend {blendOption=Linear}
    boneToModify {boneName=cape_02_r}
    lookAtLocation=<0 -25 21>
    lookAtTarget {boneReference {boneName=pelvis}}
    lookAt_Axis {axis=<1 0>}
   }
   n5 = Look At - Bone: cape_mid_05(ComponentPose=@n4, Alpha=1) {
    alphaBoolBlend {blendOption=Linear}
    boneToModify {boneName=cape_mid_05}
    lookAtLocation=<-75 -15>
    lookAtTarget {boneReference {boneName=pelvis}}
    lookAt_Axis {axis=<1 0>}
   }
   n6 = Look At - Bone: cape_06_l(ComponentPose=@n5, Alpha=1) {
    alphaBoolBlend {blendOption=Linear}
    boneToModify {boneName=cape_06_l}
    lookAtLocation=<-75 -8 -21>
    lookAtTarget {boneReference {boneName=pelvis}}
    lookAt_Axis {axis=<1 0>}
   }
   n7 = Look At - Bone: cape_06_r(ComponentPose=@n6, Alpha=1) {
    alphaBoolBlend {blendOption=Linear}
    boneToModify {boneName=cape_06_r}
    lookAtLocation=<-75 -8 21>
    lookAtTarget {boneReference {boneName=pelvis}}
    lookAt_Axis {axis=<1 0>}
   }
   n8 = Trail controller - Bone: cape_mid_07!(ComponentPose=@n7, Alpha=1, RelaxationSpeedScale:float=2, FakeVelocity:Vector="0,0,0") -> (Pose) {
    bActorSpaceFakeVel=true
    bShowTrailLocation=true
    bUsePlanarLimit=true
    chainLength=6
    debugLifeTime=0.1
    maxDeltaTime=0.0333
    relaxationSpeedScale=1.2
    alphaScaleBiasClamp {bInterpResult=true, interpSpeedDecreasing=25, interpSpeedIncreasing=25}
    baseJoint {boneName=root}
    rotationLimits [
     {limitMax=<180 180 180>}
     {limitMax=<180 180 180>}
     {limitMax=<180 180 180>}
     {limitMax=<180 180 180>}
     {limitMax=<180 180 180>}
     {limitMax=<180 180 180>}
    ]
    rotationOffsets [<0 0 0>, <0 0 0>, <0 0 0>, <0 0 0>, <0 0 0>, <0 0 0>]
    trailBone {boneName=cape_mid_07}
    trailRelaxationSpeed {editorCurveData {keys [<at 0 15 RCIM_Linear>,<at 1 6 RCIM_Linear>]}}
   }
   n9 = Trail controller - Bone: cape_08_l!(ComponentPose=@n8, Alpha=1, RelaxationSpeedScale:float=2, FakeVelocity:Vector="0,0,0") -> (Pose) {
    bActorSpaceFakeVel=true
    bShowTrailLocation=true
    bUsePlanarLimit=true
    chainLength=6
    debugLifeTime=0.1
    maxDeltaTime=0.0333
    relaxationSpeedScale=1.2
    alphaScaleBiasClamp {bInterpResult=true, interpSpeedDecreasing=25, interpSpeedIncreasing=25}
    baseJoint {boneName=root}
    rotationLimits [
     {limitMax=<180 180 180>}
     {limitMax=<180 180 180>}
     {limitMax=<180 180 180>}
     {limitMax=<180 180 180>}
     {limitMax=<180 180 180>}
     {limitMax=<180 180 180>}
    ]
    rotationOffsets [<0 0 0>, <0 0 0>, <0 0 0>, <0 0 0>, <0 0 0>, <0 0 0>]
    trailBone {boneName=cape_08_l}
    trailRelaxationSpeed {editorCurveData {keys [<at 0 15 RCIM_Linear>,<at 1 6 RCIM_Linear>]}}
   }
   n10 = Trail controller - Bone: cape_08_r!(ComponentPose=@n9, Alpha=1, RelaxationSpeedScale:float=2, FakeVelocity:Vector="0,0,0") -> (Pose) {
    bActorSpaceFakeVel=true
    bShowTrailLocation=true
    bUsePlanarLimit=true
    chainLength=6
    debugLifeTime=0.1
    maxDeltaTime=0.0333
    relaxationSpeedScale=1.2
    alphaScaleBiasClamp {bInterpResult=true, interpSpeedDecreasing=25, interpSpeedIncreasing=25}
    baseJoint {boneName=root}
    rotationLimits [
     {limitMax=<180 180 180>}
     {limitMax=<180 180 180>}
     {limitMax=<180 180 180>}
     {limitMax=<180 180 180>}
     {limitMax=<180 180 180>}
     {limitMax=<180 180 180>}
    ]
    rotationOffsets [<0 0 0>, <0 0 0>, <0 0 0>, <0 0 0>, <0 0 0>, <0 0 0>]
    trailBone {boneName=cape_08_r}
    trailRelaxationSpeed {editorCurveData {keys [<at 0 15 RCIM_Linear>,<at 1 6 RCIM_Linear>]}}
   }
   n11 = Component To Local(ComponentPose=@n10)
   n12 = Look At - Bone: cape_02_l(ComponentPose=Local To Component(LocalPose=Use cached pose 'initialPoseCache'()), Alpha=1) {
    alphaBoolBlend {blendOption=Linear}
    boneToModify {boneName=cape_02_l}
    lookAtLocation=<0 -42 -26>
    lookAtTarget {boneReference {boneName=pelvis}}
    lookAt_Axis {axis=<1 0>}
   }
   n13 = Look At - Bone: cape_03_l(ComponentPose=@n12, Alpha=1) {
    alphaBoolBlend {blendOption=Linear}
    boneToModify {boneName=cape_03_l}
    lookAtLocation=<0 -33 -26>
    lookAtTarget {boneReference {boneName=pelvis}}
    lookAt_Axis {axis=<1 0>}
   }
   n14 = Look At - Bone: cape_mid_01(ComponentPose=@n13, Alpha=1) {
    alphaBoolBlend {blendOption=Linear}
    boneToModify {boneName=cape_mid_01}
    lookAtLocation=<0 -47>
    lookAtTarget {boneReference {boneName=pelvis}}
    lookAt_Axis {axis=<1 0>}
   }
   n15 = Look At - Bone: cape_mid_02(ComponentPose=@n14, Alpha=1) {
    alphaBoolBlend {blendOption=Linear}
    boneToModify {boneName=cape_mid_02}
    lookAtLocation=<0 -37>
    lookAtTarget {boneReference {boneName=pelvis}}
    lookAt_Axis {axis=<1 0>}
   }
   n16 = Look At - Bone: cape_02_r(ComponentPose=@n15, Alpha=1) {
    alphaBoolBlend {blendOption=Linear}
    boneToModify {boneName=cape_02_r}
    lookAtLocation=<0 -42 26>
    lookAtTarget {boneReference {boneName=pelvis}}
    lookAt_Axis {axis=<1 0>}
   }
   n17 = Look At - Bone: cape_03_r(ComponentPose=@n16, Alpha=1) {
    alphaBoolBlend {blendOption=Linear}
    boneToModify {boneName=cape_03_r}
    lookAtLocation=<0 -33 26>
    lookAtTarget {boneReference {boneName=pelvis}}
    lookAt_Axis {axis=<1 0>}
   }
   n18 = Get arm_rot_bk_r() -> (arm_rot_bk_r:double)
   n19 = "Transform (Modify) Bone - Bone: cape_02_r"(ComponentPose=@n17, Alpha=@n18, Translation="-0.000,-2.000,3.244", Rotation="0,0,0", Scale="1,1,1") {
    rotationMode=BMM_Additive
    rotationSpace=BCS_BoneSpace
    translationMode=BMM_Additive
    translationSpace=BCS_BoneSpace
    alphaBoolBlend {blendOption=Linear}
    boneToModify {boneName=cape_02_r}
    translation {y=-2, z=3.244}
   }
   n20 = "Transform (Modify) Bone - Bone: cape_03_r"(ComponentPose=@n19, Alpha=@n18, Translation="0,-3,-0.000", Rotation="0,0,0", Scale="1,1,1") {
    rotationMode=BMM_Additive
    rotationSpace=BCS_BoneSpace
    translationMode=BMM_Additive
    translationSpace=BCS_BoneSpace
    alphaBoolBlend {blendOption=Linear}
    boneToModify {boneName=cape_03_r}
    translation {y=-3}
   }
   n21 = "Transform (Modify) Bone - Bone: cape_01_l"(ComponentPose=@n20, Alpha=Get arm_rot_bk_l() -> (arm_rot_bk_l:double), Translation="-2,0.000,-5", Rotation="0,0,0", Scale="1,1,1") {
    rotationMode=BMM_Additive
    rotationSpace=BCS_BoneSpace
    translationMode=BMM_Additive
    translationSpace=BCS_BoneSpace
    alphaBoolBlend {blendOption=Linear}
    boneToModify {boneName=cape_01_l}
    translation {x=-2, z=-5}
   }
   n22 = Get CapeTrailRelaxSpeedScale() -> (CapeTrailRelaxSpeedScale:float)
   n23 = Trail controller - Bone: cape_mid_07(
    ComponentPose=@n21
    Alpha=1
    RelaxationSpeedScale=@n22
    FakeVelocity=Get CapeTrailFakeVelocityMidLeft_b() -> (CapeTrailFakeVelocityMidLeft_b:Vector)
   ) {
    bActorSpaceFakeVel=true
    bShowTrailLocation=true
    bUsePlanarLimit=true
    chainLength=6
    debugLifeTime=0.1
    maxDeltaTime=0.0333
    relaxationSpeedScale=1.2
    alphaScaleBiasClamp {bInterpResult=true, interpSpeedDecreasing=25, interpSpeedIncreasing=25}
    baseJoint {boneName=root}
    planarLimits [
     {drivingBone {boneName=root}, planeTransform {location=<0 0 10>, rotation=<rot 0 0 0>, scale=<1 1 1>}}
     {drivingBone {boneName=spine_03}, planeTransform {location=<0 -20 0>, rotation=<rot 0 15 -89.9999>, scale=<1 1 1>}}
    ]
    rotationLimits [
     {limitMax=<180 180 180>}
     {limitMax=<180 180 180>}
     {limitMax=<180 180 180>}
     {limitMax=<180 180 180>}
     {limitMax=<180 180 180>}
     {limitMax=<180 180 180>}
    ]
    rotationOffsets [<0 0 0>, <0 0 0>, <0 0 0>, <0 0 0>, <0 0 0>, <0 0 0>]
    trailBone {boneName=cape_mid_07}
    trailRelaxationSpeed {editorCurveData {keys [<at 0 15 RCIM_Linear>,<at 0.5289 13.1838 RCIM_Linear>,<at 1 6 RCIM_Linear>]}}
   }
   n24 = Get CapeTrailRelaxSpeedScale()
   n25 = Trail controller - Bone: cape_08_l(
    ComponentPose=@n23
    Alpha=1
    RelaxationSpeedScale=@n24
    FakeVelocity=Get CapeTrailFakeVelocityLowLeft_b() -> (CapeTrailFakeVelocityLowLeft_b:Vector)
   ) {
    bActorSpaceFakeVel=true
    bShowTrailLocation=true
    bUsePlanarLimit=true
    chainLength=6
    debugLifeTime=0.1
    maxDeltaTime=0.0333
    relaxationSpeedScale=1.2
    alphaScaleBiasClamp {bInterpResult=true, interpSpeedDecreasing=25, interpSpeedIncreasing=25}
    baseJoint {boneName=root}
    planarLimits [
     {drivingBone {boneName=root}, planeTransform {location=<0 0 10>, rotation=<rot 0 0 0>, scale=<1 1 1>}}
     {drivingBone {boneName=spine_03}, planeTransform {location=<0 -20 0>, rotation=<rot 0 15 -89.9999>, scale=<1 1 1>}}
    ]
    rotationLimits [
     {limitMax=<180 180 180>}
     {limitMax=<180 180 180>}
     {limitMax=<180 180 180>}
     {limitMax=<180 180 180>}
     {limitMax=<180 180 180>}
     {limitMax=<180 180 180>}
    ]
    rotationOffsets [<0 0 0>, <0 0 0>, <0 0 0>, <0 0 0>, <0 0 0>, <0 0 0>]
    trailBone {boneName=cape_08_l}
    trailRelaxationSpeed {editorCurveData {keys [<at 0 15 RCIM_Linear>,<at 0.5289 13.1838 RCIM_Linear>,<at 1 6 RCIM_Linear>]}}
   }
   n26 = Trail controller - Bone: cape_08_r(
    ComponentPose=@n25
    Alpha=1
    RelaxationSpeedScale=@n24
    FakeVelocity=Get CapeTrailFakeVelocityLowRight_b() -> (CapeTrailFakeVelocityLowRight_b:Vector)
   ) {
    bActorSpaceFakeVel=true
    bShowTrailLocation=true
    bUsePlanarLimit=true
    chainLength=6
    debugLifeTime=0.1
    maxDeltaTime=0.0333
    relaxationSpeedScale=1.2
    alphaScaleBiasClamp {bInterpResult=true, interpSpeedDecreasing=25, interpSpeedIncreasing=25}
    baseJoint {boneName=root}
    planarLimits [
     {drivingBone {boneName=root}, planeTransform {location=<0 0 10>, rotation=<rot 0 0 0>, scale=<1 1 1>}}
     {drivingBone {boneName=spine_03}, planeTransform {location=<0 -20 0>, rotation=<rot 0 15 -89.9999>, scale=<1 1 1>}}
    ]
    rotationLimits [
     {limitMax=<180 180 180>}
     {limitMax=<180 180 180>}
     {limitMax=<180 180 180>}
     {limitMax=<180 180 180>}
     {limitMax=<180 180 180>}
     {limitMax=<180 180 180>}
    ]
    rotationOffsets [<0 0 0>, <0 0 0>, <0 0 0>, <0 0 0>, <0 0 0>, <0 0 0>]
    trailBone {boneName=cape_08_r}
    trailRelaxationSpeed {editorCurveData {keys [<at 0 15 RCIM_Linear>,<at 0.5289 13.1838 RCIM_Linear>,<at 1 6 RCIM_Linear>]}}
   }
   n27 = Component To Local(ComponentPose=@n26)
   n28 = Get bIsDBNO()
   n29 = "Transform (Modify) Bone - Bone: dyn_crouch_collision"(
    ComponentPose=Local To Component(LocalPose=Use cached pose 'initialPoseCache'())
    Alpha=Get crouch alpha inv() -> (crouch alpha inv:double)
    Translation="0,0.000,-49.249"
    Rotation="0,0,0"
    Scale="1,1,1"
   ) {
    translationMode=BMM_Replace
    translationSpace=BCS_BoneSpace
    alphaBoolBlend {blendOption=Linear}
    alphaScaleBiasClamp {bInterpResult=true}
    boneToModify {boneName=dyn_crouch_collision}
    translation {z=-49.249}
   }
   n30 = "Transform (Modify) Bone - Bone: cape_00_r"(ComponentPose=@n29, Alpha=Get armForwardUpR() -> (armForwardUpR:double), Translation="0,0,0", Rotation="30,0.000,0", Scale="1,1,1") {
    rotationMode=BMM_Additive
    rotationSpace=BCS_BoneSpace
    translationMode=BMM_Additive
    translationSpace=BCS_BoneSpace
    alphaBoolBlend {blendOption=Linear}
    boneToModify {boneName=cape_00_r}
    rotation {pitch=30}
   }
   n31 = "Transform (Modify) Bone - Bone: cape_00_l"(ComponentPose=@n30, Alpha=Get armForwardUpL() -> (armForwardUpL:double), Translation="0,0,0", Rotation="20,0,0", Scale="1,1,1") {
    rotationMode=BMM_Additive
    rotationSpace=BCS_BoneSpace
    translationMode=BMM_Additive
    translationSpace=BCS_BoneSpace
    alphaBoolBlend {blendOption=Linear}
    boneToModify {boneName=cape_00_l}
    rotation {pitch=20}
   }
   n32 = "Constraint - dyn_collar_02_r_MSTR (PP)"!(ComponentPose=@n31, Alpha=0, ConstraintWeights_0=0.5, ConstraintWeights_1:float=0.5) -> (Pose) {
    alphaBoolBlend {blendOption=Linear}
    boneToModify {boneName=dyn_collar_02_r_MSTR}
    constraintSetup [{targetBone {boneName=spine_05}}, {targetBone {boneName=clavicle_r}}]
    constraintWeights [1,1]
   }
   n33 = "Constraint - dyn_collar_02_l_MSTR (PP)"!(ComponentPose=@n32, Alpha=0, ConstraintWeights_0=0.5, ConstraintWeights_1:float=0.5) -> (Pose) {
    alphaBoolBlend {blendOption=Linear}
    boneToModify {boneName=dyn_collar_02_l_MSTR}
    constraintSetup [{targetBone {boneName=spine_05}}, {targetBone {boneName=clavicle_l}}]
    constraintWeights [1,1]
   }
   n34 = "Transform (Modify) Bone - Bone: cape_01_r"(ComponentPose=@n33, Alpha=Get arm_rot_bk_r(), Translation="-1.386,-1.870,4", Rotation="15,0,0", Scale="1,1,1") {
    rotationMode=BMM_Additive
    rotationSpace=BCS_BoneSpace
    translationMode=BMM_Additive
    translationSpace=BCS_BoneSpace
    alphaBoolBlend {blendOption=Linear}
    boneToModify {boneName=cape_01_r}
    rotation {pitch=15}
    translation=<-1.386 -1.87 4>
   }
   n35 = "Transform (Modify) Bone - Bone: cape_01_l"(ComponentPose=@n34, Alpha=Get arm_rot_bk_l(), Translation="-1.668,0.000,-3.263", Rotation="0,0,0", Scale="1,1,1") {
    translationMode=BMM_Additive
    translationSpace=BCS_BoneSpace
    alphaBoolBlend {blendOption=Linear}
    boneToModify {boneName=cape_01_l}
    translation {x=-1.668, z=-3.263}
   }
   n36 = RigidBody(ComponentPose=@n35, OverrideWorldGravity=@n1) {
    bClampLinearTranslationLimitToRefPose=true
    bOverrideWorldGravity=true
    overridePhysicsAsset=&M_MED_Plum_Retro_DynCape
    simulationSpace=BaseBoneSpace
    alphaBoolBlend {blendOption=Linear}
    baseBoneRef {boneName=spine_05}
    componentAppliedLinearAccClamp=<3000 3000 3000>
    componentLinearAccScale {x=0.45}
    componentLinearVelScale {x=0.45}
    overrideWorldGravity {z=-980}
    simSpaceSettings {
     maxAngularAcceleration=200
     maxAngularVelocity=5
     maxLinearAcceleration=2000
     maxLinearVelocity=200
     velocityScaleZ=0
     worldAlpha=1
     externalLinearDragV=<1 1>
    }
   }
   n37 = RigidBody(ComponentPose=@n36, OverrideWorldGravity="0,0,500") {
    bClampLinearTranslationLimitToRefPose=true
    bOverrideWorldGravity=true
    overridePhysicsAsset=&/M_MED_Plum_Retro_DynCollar
    simulationSpace=BaseBoneSpace
    alphaBoolBlend {blendOption=Linear}
    baseBoneRef {boneName=spine_05}
    componentAppliedLinearAccClamp=<3000 3000 3000>
    componentLinearAccScale=<0.2 0.2 0.2>
    componentLinearVelScale=<0.2 0.2 0.2>
    overrideWorldGravity {z=-980}
   }
   n38 = "Transform (Modify) Bone - Bone: cape_01_l"(ComponentPose=@n37, Alpha=Get arm_rot_bk_l(), Translation="0,-3,0", Rotation="0,0,0", Scale="1,1,1") {
    translationMode=BMM_Additive
    translationSpace=BCS_BoneSpace
    alphaBoolBlend {blendOption=Linear}
    boneToModify {boneName=cape_01_l}
    translation {y=-3}
   }
   n39 = "Transform (Modify) Bone - Bone: dyn_collar_02_r_MSTR"!(ComponentPose=@n38, bAlphaBoolEnabled:bool=Get bIsInFrontEnd(), Translation="0,0,0", Rotation="20.398,0.000,0.000", Scale="1,1,1") -> (Pose) {
    alphaInputType=Bool
    rotationMode=BMM_Additive
    rotationSpace=BCS_BoneSpace
    alphaBoolBlend {blendOption=Linear}
    boneToModify {boneName=dyn_collar_02_r_MSTR}
    rotation {pitch=20.398}
   }
   Output Pose(Result=Blend Poses by bool(
     BlendPose_0=Component To Local(ComponentPose=@n39)
     BlendPose_1=Blend Poses by bool(BlendPose_0=@n11, BlendPose_1=@n27, bActiveValue=@n28) {blendType=Linear}
     bActiveValue=Get bRigidBodyEnabled()
    ) {blendType=Linear}) {layerGroup=None}
   Save cached pose 'initialPoseCache'(Pose=Copy Pose From Mesh() {bUseAttachedParent=true}) {cacheName=initialPoseCache}
   Trail controller - Bone: cape_mid_07_r!(
    ComponentPose
    Alpha=1
    RelaxationSpeedScale:float=@n22
    FakeVelocity:Vector=Get CapeTrailFakeVelocityMidRight_b() -> (CapeTrailFakeVelocityMidRight_b:Vector)
   ) -> (Pose) {
    bActorSpaceFakeVel=true
    bShowTrailLocation=true
    bUsePlanarLimit=true
    chainLength=6
    debugLifeTime=0.1
    maxDeltaTime=0.0333
    relaxationSpeedScale=1.2
    alphaScaleBiasClamp {bInterpResult=true, interpSpeedDecreasing=25, interpSpeedIncreasing=25}
    baseJoint {boneName=root}
    planarLimits [
     {drivingBone {boneName=root}, planeTransform {location=<0 0 10>, rotation=<rot 0 0 0>, scale=<1 1 1>}}
     {drivingBone {boneName=spine_03}, planeTransform {location=<0 -20 0>, rotation=<rot 0 15 -89.9999>, scale=<1 1 1>}}
    ]
    rotationLimits [
     {limitMax=<180 180 180>}
     {limitMax=<180 180 180>}
     {limitMax=<180 180 180>}
     {limitMax=<180 180 180>}
     {limitMax=<180 180 180>}
     {limitMax=<180 180 180>}
    ]
    rotationOffsets [<0 0 0>, <0 0 0>, <0 0 0>, <0 0 0>, <0 0 0>, <0 0 0>]
    trailBone {boneName=cape_mid_07_r}
    trailRelaxationSpeed {editorCurveData {keys [<at 0 15 RCIM_Linear>,<at 0.5289 13.1838 RCIM_Linear>,<at 1 6 RCIM_Linear>]}}
   }
  ]}
 EventGraph {nodes [
   Event BlueprintUpdateAnimation()
   > n1 = Sequence()
   n2 = GetPartSkeletalMeshComponent()
   > n3 = TransformFromBoneSpace!(
    execute
    self:Skinned Mesh Component Object Reference=@n2
    BoneName:Name=lowerarm_r
    InPosition:Vector="0, 0, 0"
    InRotation:Rotator="0, 0, 0"
   ) -> (then, OutPosition:Vector, OutRotation:Rotator)
   > n4 = TransformFromBoneSpace(self=@n2, BoneName=head, InPosition="0, 0, 0", InRotation="0, 0, 0")
   > Set armForwardUpR(armForwardUpR:double=MapRangeClamped(Value="Distance (Vector)"!(V1:Vector=@n3, V2:Vector=@n4) -> (ReturnValue:double), InRangeA=40, InRangeB=20)) -> (Output_Get:double)
   n5 = GetPartSkeletalMeshComponent()
   > n6 = TransformFromBoneSpace(self=@n5, BoneName=lowerarm_l, InPosition="0, 0, 0", InRotation="0, 0, 0")
   > n7 = TransformFromBoneSpace(self=@n5, BoneName=head, InPosition="0, 0, 0", InRotation="0, 0, 0")
   > Set armForwardUpL(armForwardUpL:double=MapRangeClamped(Value="Distance (Vector)"(V1=@n6, V2=@n7), InRangeA=45, InRangeB=30)) -> (Output_Get:double)
   n8 = GetOwnerSkeletalMeshComponent()
   n9 = BreakRotator(InRot=BreakTransform(InTransform=GetDeltaTransformFromRefPose(self=@n8, BoneName=upperarm_r, BaseName=spine_04)).Rotation)
   Set arm_rot_bk_r(execute=@n1.then_1, arm_rot_bk_r:double=MapRangeClamped(Value="Absolute (Float)"(A=@n9.Yaw), InRangeA=0, InRangeB=30)) -> (Output_Get:double)
   n10 = BreakTransform(InTransform=GetDeltaTransformFromRefPose(self=GetOwnerSkeletalMeshComponent(), BoneName=upperarm_l, BaseName=spine_04))
   > Set arm_rot_bk_l(arm_rot_bk_l:double=MapRangeClamped(Value=BreakRotator(InRot=@n10.Rotation).Yaw, InRangeA=5, InRangeB=-15)) -> (Output_Get:double)
   Event BlueprintInitializeAnimation()
   n11 = TryGetPawnOwner()
   > Is Valid(InputObject=@n11)
   > n12 = Cast To FortPlayerPawnAthena(Object=@n11)
   > Set player_pawn(player_pawn:Fort Player Pawn Athena Object Reference=@n12) -> (Output_Get:Fort Player Pawn Athena Object Reference)
   Event OnEndCrouching()
   > Set crouch alpha(crouch alpha:double=0) -> (Output_Get:double)
   > Set crouch alpha inv(crouch alpha inv:double=1) -> (Output_Get:double)
   Event OnBeginCrouching()
   > Set crouch alpha(crouch alpha=1)
   > Set crouch alpha inv(crouch alpha inv=0)
   TryGetPawnOwner()
  ]}
]
variables [
 arm_rot_bk_l {type=number}
 arm_rot_bk_r {type=number}
 armForwardUpL {type=number}
 armForwardUpR {type=number}
 CapeTrailFakeVelocityLowLeft_b {type=Vector}
 CapeTrailFakeVelocityLowRight_b {type=Vector}
 CapeTrailFakeVelocityMidLeft_b {type=Vector}
 CapeTrailFakeVelocityMidRight_b {type=Vector}
 crouch alpha {type=number}
 crouch alpha inv {type=number}
 player_pawn {type=FortPlayerPawnAthena}
]
